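# Translucent lit smoke trail material for MrMite emote FX, usable on beam trails, Niagara ribbons and particle sprites.
# A panned texture, eroded via dynamic parameters, drives opacity with end-gradient, camera, depth and view-angle fades.
# Mobile skips those fades except the gradient, and an optional world-space sphere mask can scale opacity.
# Emissive combines particle color, an optional invertible color texture, erosion and EmissiveAmount through HDR remapping.
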Material /BRCosmetics/Animation/Game/MainPlayer/Emotes/MrMite/FX/M_Smoke_Trail_Soft
graph {
 properties {
  bUsedWithBeamTrails=true
  bUsedWithNiagaraRibbons=true
  bUsedWithParticleSprites=true
  blendMode=BLEND_Translucent
  pixelDepthOffsetMode=PDOM_Legacy
  usedShadingModels=Default Lit
 }
 nodes [
  n1 = ParticleColor()
  n2 = DynamicParameter() {defaultValue=<rgba 1 0 1 1>, paramNames [Erode, EmissiveAmount, PanSpeed, MasterAlpha]}
  n3 = Panner(
   Coordinate=TextureCoordinate() {coordinateIndex=1}
   Time=Multiply(A=@n2.PanSpeed, B=Time())
   Speed=VectorParameter() {parameterName=PanSpeed, defaultValue=<rgba 0.6 0 0 1>}.RGB
  ) {speedX=0.6}
  n4 = TextureSampleParameter2D(UVs=@n3) {parameterName=TrailTexture_BChannel, samplerType=SAMPLERTYPE_Grayscale, texture=&T_Smoke_Trail_Soft}
  n5 = Add(A=@n2.Erode, B=@n4.G)
  n6 = Clamp(None=Multiply(A=LinearInterpolate(Alpha=@n5) {constA=-3}, B=@n4.G))
  n7 = Multiply(A=@n1.A, B=@n6)
  n8 = MaterialFunctionCall(
   Fade Length=ScalarParameter() {defaultValue=350, parameterName=Fade Length}
   Fade Offset=ScalarParameter() {defaultValue=24, parameterName=Fade Offset}
  ) {materialFunction=&CameraDepthFade}
  n9 = MaterialFunctionCall() {materialFunction=&LinearGradient}
  n10 = Multiply(A=OneMinus(None=@n9.UGradient), B=@n9.UGradient)
  n11 = Saturate(None=MaterialFunctionCall(In=Multiply(A=@n10) {constB=8}, Contrast=ScalarParameter() {defaultValue=0.1, parameterName=TailGradientFallOff}) {materialFunction=&CheapContrast}.Result)
  n12 = DynamicParameter() {desc=Index 1 Params for Niagara, parameterIndex=1, defaultValue=<rgba 30 1 1 1>, paramNames [DepthFadeDist, Param2, Param3, Param4]}
  n13 = ScalarParameter() {defaultValue=40, parameterName=DepthFade}
  n14 = DepthFade(
   Opacity=Multiply(A=Multiply(A=@n7, B=@n8.Result), B=@n11)
   FadeDistance=If(A=@n12.DepthFadeDist, B=Constant() {r=1}, "A > B"=@n12.DepthFadeDist, "A == B"=@n13, "A < B"=@n13)
  ) {fadeDistanceDefault=40}
  n15 = MaterialFunctionCall("FadeSensitivity (optional)"=ScalarParameter() {parameterName=EdgeFadeSensitivity}) {materialFunction=&FadeBasedOnViewAngle}
  n16 = ShadingPathSwitch(
   Default=Multiply(
    A=Multiply(A=@n14, B=@n15.Result)
    B=DynamicParameter() {defaultValue=<rgba 1 0 1 1>, paramNames [Erode, EmissiveAmount, PanSpeed, MasterAlpha]}.MasterAlpha
   )
   Mobile=Multiply(A=@n7, B=@n11)
  )
  n17 = ScalarParameter() {group="4.2) Emissive Sphere Mask", parameterName=Sphere Mask Low Value, sortPriority=1}
  n18 = ScalarParameter() {defaultValue=1, group="4.2) Emissive Sphere Mask", parameterName=Sphere Mask High Value, sortPriority=2}
  n19 = ObjectPositionWS()
  n20 = VectorParameter() {group="4.2) Emissive Sphere Mask", parameterName=Location Offset, sortPriority=5, defaultValue=<rgba 0 0 0 1>}
  n21 = Multiply(
   A=SphereMask(
    A=WorldPosition().XYZ
    B=Add(A=@n19, B=@n20.RGB)
    Radius=ScalarParameter() {defaultValue=50, group="4.2) Emissive Sphere Mask", parameterName=Sphere Mask Radius, sortPriority=3}
    Hardness=ScalarParameter() {group="4.2) Emissive Sphere Mask", parameterName=Sphere Mask Hardness, sortPriority=4}
   )
   B=ScalarParameter() {group="4.2) Emissive Sphere Mask", parameterName=Hardness, sortPriority=6}
  )
  n22 = Saturate(None=Power(Base=Clamp(None=@n21), Exp=ScalarParameter() {defaultValue=1, group="4.2) Emissive Sphere Mask", parameterName=Power, sortPriority=5}) {constExponent=1})
  LinearInterpolate(
   A=@n16
   B=Multiply(A=Saturate(None=@n16), B=LinearInterpolate(A=@n17, B=@n18, Alpha=@n22))
   Alpha=ScalarParameter() {group="4.2) Emissive Sphere Mask", parameterName="Use Sphere Mask (World)", sortPriority=-1}
  )
  n23 = TextureSampleParameter2D(UVs=@n3) {parameterName=ColorTexture, texture=/Game/Effects/Fort_Effects/Textures/Smoke/T_Smoke_Trail_Soft_Chromatic}
  n24 = OneMinus(None=@n23.RGB)
  n25 = ScalarParameter() {parameterName=InvertColorMap}
  n26 = Multiply(
   A=@n1.RGB
   B=StaticSwitchParameter(True=LinearInterpolate(A=@n23.RGB, B=@n24, Alpha=@n25), False=Constant() {r=1}) {parameterName=UseColorTexture}
  )
  n27 = Multiply(A=DynamicParameter() {defaultValue=<rgba 1 0 1 1>, paramNames [Erode, EmissiveAmount, PanSpeed, MasterAlpha]}.EmissiveAmount, B=@n6)
  MaterialFunctionCall(Emissive=MaterialFunctionCall(LDR Input=Saturate(None=@n26), HDR Tint=Multiply(A=@n26, B=@n27)) {materialFunction=&DeriveHDRfromLDR}.RemappedRange) {materialFunction=&MF_FX_EmissiveAdjustments}
  Reroute(None=@n1.RGB)
 ]
}
