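# Niagara system for the smoke or steam effect of the MrMite emote.
# Only Steam002 is enabled; two other steam sprite emitters and a MugSmoke001 ribbon emitter are disabled.
# Steam002 bursts and continuously spawns short-lived, growing SubUV sprites, slowed by drag, faded by System.FadeIn and tinted by time of day.
# BlendSpace raises the world/local position blend from 0.8 to 1 based on an owner-speed threshold.
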NiagaraSystem /BRCosmetics/Animation/Game/MainPlayer/Emotes/MrMite/FX/NS_Emote_MrMite_Smoke
dependencies {
 usedDataInterfaces [&NiagaraDataInterfaceCurve, &NiagaraDataInterfaceColorCurve]
 usedDynamicInputs []
 usedModules [
  &self:SystemSpawnScript.SystemScriptSource.SystemScriptGraph.NiagaraNodeAssignment_0.SetVariables_#1
  &self:SystemSpawnScript.SystemScriptSource.SystemScriptGraph.NiagaraNodeAssignment_2.SetVariables_#2
  &self:SystemSpawnScript.SystemScriptSource.SystemScriptGraph.NiagaraNodeAssignment_3.SetVariables_#3
  &SystemState
  &BlendSpaceHelper
  &self:SystemSpawnScript.SystemScriptSource.SystemScriptGraph.NiagaraNodeAssignment_1.SetVariables_#4
 ]
}
emitters {
 MugSmoke001 {
  bEnabled=false
  emitterSpawnScript [/SetVariables(Emitter.SystemRand01:NiagaraFloat=&RandomRangeFloat(Minimum=0, Maximum=1)) {isSetParameters=true}]
  emitterUpdateScript [
   &EmitterState(Life Cycle Mode=::NewEnumerator1, Loop Duration=1, UseLoopDelay=false)
   &SpawnRate(SpawnRate=33, Use Spawn Probability=false)
  ]
  particleSpawnScript [
   &InitializeParticle(
    Color=$User.BaseColorAlpha
    Lifetime=0.8
    Ribbon Facing Mode=::NewEnumerator1
    Ribbon Facing Vector:Vector3f=<1 0 0>
    Ribbon Width=&Multiply_Float(A=15, B=$User.RibbonScale)
    Ribbon Width Mode=::NewEnumerator1
    UsePositionOffset=false
    Write Material Random=true
   )
   &SystemLocation(
    MaskSpawnGroup=false
    Offset=&Lerp_Vector(
     A=<17.5666 0.6719 -6.7835>
     B=&Lerp_Vector(A=<17.1895 0.2585 -6.6443>, B=<18.475 -4.8222 -7.3953>, Alpha=$User.Smoke_Large_Offset)
     Alpha=$User.Smoke_Large_Offset
    )
   ) {enabled=false}
   &SystemLocation(MaskSpawnGroup=false, Offset=&AddVectorToPosition(Position=$Particles.Position, Vector=<0 0 -5>)) {module=SystemLocation001}
   &AddVelocity(Velocity=<0 0 60>)
  ]
  particleUpdateScript [
   &ParticleState()
   &ScaleColor(
    Scale Alpha=&Multiply_Float(
     A=&FloatFromCurve(
      FloatCurve=DataInterface() {propertyValues {Curve {keys [<at 0 0>, {arriveTangent=-1.7714, interpMode=RCIM_Cubic, leaveTangent=-1.7714, tangentMode=RCTM_User, time=0.5, value=1}, <at 0.8 0>]}}}
      Scale Curve=&FNTimeOfDayScalar(Night Scale=1, Morning Scale=0.3, Evening Scale=0.5)
     )
     B=&FloatFromCurve(FloatCurve=DataInterface() {propertyValues {Curve {keys [<at 0 1>,<at 0.3 1>,<at 0.85 1>,<at 1 0>]}}})
    )
    Scale RGB=&MakeVectorFromLinearColorRGB(Color=&MakeCustomLinearColorFromBool(
      Bool=$User.UseEmissiveSmoke
      True Color=&LerpLinearColors(
       StartColor=$User.EmissiveColor
       EndColor=&MultiplyLinearColor(Linear Color=$User.EmissiveColor, Scale Factor=&MakeVector4_FromFloat(Float=0.25))
       LerpFactor=&FloatFromCurve(FloatCurve=DataInterface() {propertyValues {Curve {keys [<at 0 0 RCIM_Cubic>, {arriveTangent=2.2656, interpMode=RCIM_Cubic, leaveTangent=2.2656, tangentMode=RCTM_User, time=1, value=1}]}}})
      )
      False Color=<rgba 1 1 1 1>
     ))
    ScaleA=true
    ScaleRGB=true
   )
   &DynamicMaterialParameters(
    Index 0 Param 1=&FloatFromCurve(FloatCurve=DataInterface() {propertyValues {Curve {keys [
         <at 0 1 RCIM_Cubic>
         {arriveTangent=-1.7387, arriveTangentWeight=1.9926, interpMode=RCIM_Cubic, leaveTangent=-1.7387, tangentMode=RCTM_User, time=1, value=0.125}
        ]}}})
    Index 0 Param 3=-1
    Index 0 Param 4=&MakeCustomFloatFromBool(
     Bool=$User.UseEmissiveSmoke
     True Float=&FloatFromCurve(FloatCurve=DataInterface() {propertyValues {Curve {keys [{arriveTangent=-2.94, interpMode=RCIM_Cubic, time=0, value=1}, <at 0.75 0 RCIM_Cubic>]}}})
     False Float=0
    )
    Index 1 Param 1=&FloatFromCurve(FloatCurve=DataInterface() {propertyValues {Curve {keys [<at 0 1 RCIM_Cubic>]}}})
    Index 1 Param 2=&FloatFromCurve(FloatCurve=DataInterface() {propertyValues {Curve {keys [{arriveTangent=-2.2405, interpMode=RCIM_Cubic, leaveTangent=-2.2405, tangentMode=RCTM_User, time=0, value=1}, <at 1 0 RCIM_Cubic>]}}})
    Index 1 Param 3=$Emitter.SystemRand01
    Module.Index0Float0Write=true
    Module.Index0Float1Write=true
    Module.Index0Float2Write=true
    Module.Index0Float3Write=true
    Module.Index1Float0Write=true
    Module.Index1Float1Write=true
    Module.Index1Float2Write=true
    Module.Index1Float3Write=true
    Write Parameter Index 1=::NewEnumerator1
   )
   &ScaleRibbonWidth(Ribbon Width Scale=&FloatFromCurve(FloatCurve=DataInterface() {propertyValues {Curve {keys [{tangentMode=RCTM_User, time=0, value=1}, {tangentMode=RCTM_User, time=1, value=0}]}}}))
   &InheritVelocity(Inherited Velocity Speed Limit=-2, Source Speed Threshold=12, Speed Limit=true, UseSpeedThreshold=true)
   &Drag(Drag=2, Rotational Drag=2, Use Linear Drag=true, Use Rotational Drag=true)
   &SolveForcesAndVelocity(Clamp Velocity=false, Limit Acceleration=false)
   &ScaleColor(ScaleA=true, ScaleRGB=true) {module=ScaleColor001}
   &ScaleColor(Scale Alpha=$User.Global_Smoke_Opacity_Multiplier, ScaleA=true, ScaleRGB=true) {enabled=false, module=ScaleColor002}
  ]
  rendererClasses [&NiagaraRibbonRendererProperties]
  renderers [{rendererClass=&NiagaraRibbonRendererProperties}]
 }
 Steam {
  bEnabled=false
  emitterUpdateScript [
   &EmitterState()
   &SpawnBurst_Instantaneous(Spawn Count=10, Use Spawn Probability=false, UseLoopCountLimit=false)
   &SpawnRate(SpawnRate=12, Use Spawn Probability=false)
  ]
  particleSpawnScript [
   &InitializeParticle(
    Lifetime Max=0.6
    Lifetime Mode=::NewEnumerator1
    Position Offset=<0 0 5>
    Sprite Rotation Mode=::NewEnumerator2
    Sprite Size Mode=::NewEnumerator1
    Uniform Sprite Size Max=22
    Uniform Sprite Size Min=35
    UsePositionOffset=true
    Write Material Random=true
   )
   &AddVelocity(Velocity=&RandomRangeVector(Minimum=<-1.5 -1.5 3>, Maximum=<1.5 1.5 1>, Evaluation Type=::NewEnumerator1))
   &ShapeLocation(
    Cylinder Height=2
    Cylinder Radius=2
    MaskSpawnGroup=false
    Override Seed=false
    Shape Primitive=::NewEnumerator1
    Surface Only=false
    UseOwnerScale=true
   )
  ]
  particleUpdateScript [
   &ParticleState()
   &ScaleSpriteSize(Scale Factor=&Vector2DFromFloat(Value=&FloatFromCurve(FloatCurve=DataInterface() {propertyValues {Curve {keys [
          {arriveTangent=1.8751, interpMode=RCIM_Cubic, leaveTangent=1.8751, tangentMode=RCTM_User, time=0, value=0.25}
          {arriveTangent=0.72, interpMode=RCIM_Cubic, leaveTangent=0.72, tangentMode=RCTM_User, time=1, value=1.25}
         ]}}})))
   &SubUVAnimation(
    Sub UV Play Rate:NiagaraFloat=&RandomRangeFloat(Minimum=0.01, Maximum=0.01)
    SubUV Animation Mode:ENiagara_SubUVLookupMode=::NewEnumerator3
    SubUV Start Time Offset:NiagaraFloat=&RandomRangeFloat(Minimum=0, Maximum=63)
   )
   &DynamicMaterialParameters(
    Index 0 Param 1=&FloatFromCurve(FloatCurve=DataInterface() {propertyValues {Curve {keys [<at 0 -0.25>,<at 1 -1>]}}})
    Index 0 Param 4=&OneMinusFloat(Float=$Particles.NormalizedAge)
    Module.Param0WriteEnabled=true
    Module.Param1WriteEnabled=false
    Module.Param2WriteEnabled=false
    Module.Param3WriteEnabled=true
   )
   &ScaleColor(Scale Alpha=$System.FadeIn, ScaleA=true, ScaleRGB=false) {module=ScaleColor001}
   &Drag(
    Drag=&FloatFromCurve(
     FloatCurve=DataInterface() {propertyValues {Curve {
        defaultValue=1
        keys [
         {interpMode=RCIM_Cubic, tangentMode=RCTM_User, time=0, value=0}
         {arriveTangent=0.874, interpMode=RCIM_Cubic, leaveTangent=0.874, tangentMode=RCTM_User, time=1, value=1}
        ]
       }}}
     Scale Curve=&RandomRangeFloat(Minimum=3, Maximum=6)
    )
    Use Linear Drag=true
    Use Rotational Drag=false
   )
   &SolveForcesAndVelocity(Clamp Velocity=false, Limit Acceleration=false)
   &ScaleColor(
    Linear Color Curve=DataInterface() {propertyValues {
      AlphaCurve {keys [<at 0 0>,<at 0.3267 0.8>,<at 0.7075 0.4>,<at 1 0>]}
      BlueCurve {keys [<at 0 0.3478>,<at 0.6134 0.4117>,<at 1 1>]}
      GreenCurve {keys [<at 0 0.3653>,<at 0.6134 0.4203>,<at 1 1>]}
      RedCurve {keys [<at 0 0.3698>,<at 0.6134 0.499>,<at 1 1>]}
     }}
    Scale Mode=::NewEnumerator2
   ) {module=ScaleColor002}
   &ScaleColor(
    Scale RGB=&MakeVectorFromLinearColorRGB(Color=&FNTimeOfDayColor(
      Night Color=<rgba 0.4404 0.5421 0.6698 1>
      Morning Color=<rgba 1 0.8859 0.7875 1>
      Day Color=<rgba 1 1 1 1>
      Evening Color=<rgba 0.9352 0.835 1 1>
     ))
    ScaleA=false
    ScaleRGB=true
   )
  ]
  rendererClasses [&NiagaraSpriteRendererProperties]
  renderers [{rendererClass=&NiagaraSpriteRendererProperties}]
 }
 Steam001 {
  bEnabled=false
  emitterUpdateScript [
   &EmitterState()
   &SpawnBurst_Instantaneous(Spawn Count=10, Use Spawn Probability=false, UseLoopCountLimit=false)
   &SpawnRate(SpawnRate=22, Use Spawn Probability=false)
  ]
  particleSpawnScript [
   &InitializeParticle(
    Lifetime Max=0.6
    Lifetime Min=0.5
    Lifetime Mode=::NewEnumerator1
    Sprite Rotation Mode=::NewEnumerator2
    Sprite Size Mode=::NewEnumerator1
    Uniform Sprite Size Max=22
    Uniform Sprite Size Min=35
    UsePositionOffset=true
    Write Material Random=true
   )
   &AddVelocity(Velocity=&RandomRangeVector(Minimum=<-2 -2 3>, Maximum=<2 2 23>, Evaluation Type=::NewEnumerator1))
   &ShapeLocation(
    Cylinder Height=2
    Cylinder Radius=2
    MaskSpawnGroup=false
    Override Seed=false
    Shape Primitive=::NewEnumerator1
    Surface Only=false
    UseOwnerScale=true
   )
  ]
  particleUpdateScript [
   &ParticleState()
   &ScaleSpriteSize(Scale Factor=&Vector2DFromFloat(Value=&FloatFromCurve(FloatCurve=DataInterface() {propertyValues {Curve {keys [
          {arriveTangent=1.8751, interpMode=RCIM_Cubic, leaveTangent=1.8751, tangentMode=RCTM_User, time=0, value=0.25}
          {arriveTangent=0.72, interpMode=RCIM_Cubic, leaveTangent=0.72, tangentMode=RCTM_User, time=1, value=1.25}
         ]}}})))
   &SubUVAnimation(
    Sub UV Play Rate:NiagaraFloat=&RandomRangeFloat(Minimum=0.01, Maximum=0.1)
    SubUV Animation Mode:ENiagara_SubUVLookupMode=::NewEnumerator3
    SubUV Start Time Offset:NiagaraFloat=&RandomRangeFloat(Minimum=0, Maximum=63)
   )
   &DynamicMaterialParameters(
    Index 0 Param 1=&FloatFromCurve(FloatCurve=DataInterface() {propertyValues {Curve {keys [<at 0 -0.25>,<at 1 -1>]}}})
    Index 0 Param 4=&OneMinusFloat(Float=$Particles.NormalizedAge)
    Module.Param0WriteEnabled=true
    Module.Param1WriteEnabled=false
    Module.Param2WriteEnabled=false
    Module.Param3WriteEnabled=true
   )
   &ScaleColor(Scale Alpha=$System.FadeIn, ScaleA=true, ScaleRGB=false) {module=ScaleColor001}
   &ScaleColor(
    Scale Alpha=&FloatFromCurve(
     FloatCurve=DataInterface() {propertyValues {Curve {keys [
         {arriveTangent=5.2427, interpMode=RCIM_Cubic, leaveTangent=5.2427, tangentMode=RCTM_User, time=0, value=0}
         {arriveTangent=-1.0898, interpMode=RCIM_Cubic, leaveTangent=-1.0898, tangentMode=RCTM_User, time=0.3, value=1}
         {arriveTangent=-0.1231, interpMode=RCIM_Cubic, leaveTangent=-0.1231, tangentMode=RCTM_User, time=1, value=0.5}
        ]}}}
     Scale Curve=&RandomRangeFloat(Minimum=0.5, Maximum=1)
    )
    ScaleA=true
    ScaleRGB=false
   ) {enabled=false}
   &Drag(
    Drag=&FloatFromCurve(
     FloatCurve=DataInterface() {propertyValues {Curve {
        defaultValue=1
        keys [
         {interpMode=RCIM_Cubic, tangentMode=RCTM_User, time=0, value=0}
         {arriveTangent=0.874, interpMode=RCIM_Cubic, leaveTangent=0.874, tangentMode=RCTM_User, time=1, value=1}
        ]
       }}}
     Scale Curve=&RandomRangeFloat(Minimum=3, Maximum=6)
    )
    Use Linear Drag=true
    Use Rotational Drag=false
   )
   &SolveForcesAndVelocity(Clamp Velocity=false, Limit Acceleration=false)
   &ScaleColor(
    Linear Color Curve=DataInterface() {propertyValues {
      AlphaCurve {keys [<at 0 0>,<at 0.2176 0.7>,<at 0.5517 0.4>,<at 1 0>]}
      BlueCurve {keys [<at 0 0.2259>,<at 0.629 0.4826>,<at 1 1>]}
      GreenCurve {keys [<at 0 0.2561>,<at 0.629 0.4798>,<at 1 1>]}
      RedCurve {keys [<at 0 0.2948>,<at 0.629 0.499>,<at 1 1>]}
     }}
    Scale Mode=::NewEnumerator2
   ) {module=ScaleColor002}
  ]
  rendererClasses [&NiagaraSpriteRendererProperties]
  renderers [{rendererClass=&NiagaraSpriteRendererProperties}]
 }
 Steam002 {
  emitterUpdateScript [
   &EmitterState()
   &SpawnBurst_Instantaneous(Spawn Count=10, Use Spawn Probability=false, UseLoopCountLimit=false)
   &SpawnRate(SpawnRate=12, Use Spawn Probability=false)
  ]
  particleSpawnScript [
   &InitializeParticle(
    Lifetime Max=0.6
    Lifetime Mode=::NewEnumerator1
    Position Offset=<0 0 5>
    Sprite Rotation Mode=::NewEnumerator2
    Sprite Size Mode=::NewEnumerator1
    Uniform Sprite Size Max=22
    Uniform Sprite Size Min=35
    UsePositionOffset=true
    Write Material Random=true
   )
   &AddVelocity(Velocity=&RandomRangeVector(Minimum=<-1.5 -1.5 3>, Maximum=<1.5 1.5 1>, Evaluation Type=::NewEnumerator1))
   &ShapeLocation(
    Cylinder Height=2
    Cylinder Radius=2
    MaskSpawnGroup=false
    Override Seed=false
    Shape Primitive=::NewEnumerator1
    Surface Only=false
    UseOwnerScale=true
   )
  ]
  particleUpdateScript [
   &ParticleState()
   &ScaleSpriteSize(Scale Factor=&Vector2DFromFloat(Value=&FloatFromCurve(FloatCurve=DataInterface() {propertyValues {Curve {keys [
          {arriveTangent=1.8751, interpMode=RCIM_Cubic, leaveTangent=1.8751, tangentMode=RCTM_User, time=0, value=0.25}
          {arriveTangent=0.72, interpMode=RCIM_Cubic, leaveTangent=0.72, tangentMode=RCTM_User, time=1, value=1.25}
         ]}}})))
   &SubUVAnimation(
    Sub UV Play Rate:NiagaraFloat=&RandomRangeFloat(Minimum=0.01, Maximum=0.01)
    SubUV Animation Mode:ENiagara_SubUVLookupMode=::NewEnumerator3
    SubUV Start Time Offset:NiagaraFloat=&RandomRangeFloat(Minimum=0, Maximum=63)
   )
   &DynamicMaterialParameters(
    Index 0 Param 1=&FloatFromCurve(FloatCurve=DataInterface() {propertyValues {Curve {keys [<at 0 -0.25>,<at 1 -1>]}}})
    Index 0 Param 4=&OneMinusFloat(Float=$Particles.NormalizedAge)
    Module.Param0WriteEnabled=true
    Module.Param1WriteEnabled=false
    Module.Param2WriteEnabled=false
    Module.Param3WriteEnabled=true
   )
   &ScaleColor(Scale Alpha=$System.FadeIn, ScaleA=true, ScaleRGB=false) {module=ScaleColor001}
   &Drag(
    Drag=&FloatFromCurve(
     FloatCurve=DataInterface() {propertyValues {Curve {
        defaultValue=1
        keys [
         {interpMode=RCIM_Cubic, tangentMode=RCTM_User, time=0, value=0}
         {arriveTangent=0.874, interpMode=RCIM_Cubic, leaveTangent=0.874, tangentMode=RCTM_User, time=1, value=1}
        ]
       }}}
     Scale Curve=&RandomRangeFloat(Minimum=3, Maximum=6)
    )
    Use Linear Drag=true
    Use Rotational Drag=false
   )
   &SolveForcesAndVelocity(Clamp Velocity=false, Limit Acceleration=false)
   &ScaleColor(
    Linear Color Curve=DataInterface() {propertyValues {
      AlphaCurve {keys [<at 0 0>,<at 0.1504 0.6>,<at 0.4971 0.4>,<at 1 0>]}
      BlueCurve {keys [<at 0 0.5906>,<at 0.6134 0.6466>,<at 1 1>]}
      GreenCurve {keys [<at 0 0.5906>,<at 0.6134 0.6493>,<at 1 1>]}
      RedCurve {keys [<at 0 0.5906>,<at 0.6134 0.674>,<at 1 1>]}
     }}
    Scale Mode=::NewEnumerator2
   ) {module=ScaleColor002}
   &ScaleColor(
    Scale RGB=&MakeVectorFromLinearColorRGB(Color=&FNTimeOfDayColor(
      Night Color=<rgba 0.2058 0.3111 0.6698 1>
      Morning Color=<rgba 1 0.8859 0.7875 1>
      Day Color=<rgba 1 1 1 1>
      Evening Color=<rgba 1 0.668 0.6852 1>
     ))
    ScaleA=false
    ScaleRGB=true
   )
   &BlendSpace("Position Blend (World <-> Local)":NiagaraFloat=&MakeCustomFloatFromBool(
     Bool=&SetBoolByFloatComparison(A=&VectorLength(Vector=$Engine.Owner.Velocity), B=400)
     True Float=1
     False Float=0.8
     Translation Contribution:NiagaraFloat=1
     Rotation Contribution:NiagaraFloat=1
    ))
  ]
  rendererClasses [&NiagaraSpriteRendererProperties]
  renderers [{rendererClass=&NiagaraSpriteRendererProperties}]
 }
}
summary {
 systemName=NS_Emote_MrMite_Smoke
 emitters [
  {bEnabled=false, emitterName=Steam, rendererClasses [&NiagaraSpriteRendererProperties]}
  {bEnabled=false, emitterName=Steam001, rendererClasses [&NiagaraSpriteRendererProperties]}
  {bEnabled=false, emitterName=MugSmoke001, rendererClasses [&NiagaraRibbonRendererProperties]}
  {emitterName=Steam002, rendererClasses [&NiagaraSpriteRendererProperties]}
 ]
 userVariables [
  User.BaseColorAlpha {type=LinearColor, description="", defaultValue {struct=LinearColor, value=<rgba 1 1 1 0.15>}}
  User.bIsFrontEndPreview {type=NiagaraBool}
  User.EmissiveColor {type=LinearColor, description="", defaultValue {struct=LinearColor, value=<rgba 0.1354 0.0043 0 1>}}
  User.Global_Smoke_Opacity_Multiplier {type=NiagaraFloat, defaultValue=0, description=""}
  User.Loop {type=NiagaraFloat, defaultValue=3.1469, description=""}
  User.RibbonIDOffset {type=NiagaraInt32}
  User.RibbonLifetime {type=NiagaraFloat, defaultValue=0.5, description=""}
  User.RibbonMaterial {
   type=MaterialInterface
   description=""
   defaultValue {
    struct=NiagaraExt_VariableValue_Object
    value {object=/Game/Effects/Niagara/MasterMaterials/Reload/M_Reload_SmokeTrail, objectClass=/Script/Engine.MaterialInterface}
   }
  }
  User.RibbonScale {type=NiagaraFloat, defaultValue=1, description=""}
  User.Smoke_Alpha {type=NiagaraFloat, defaultValue=2, description=""}
  User.Smoke_Large_Offset {type=NiagaraFloat, defaultValue=0, description=""}
  User.Smoke_Position_Offset {type=NiagaraFloat, defaultValue=0, description=""}
  User.SocketNumber {type=NiagaraInt32}
  User.UseEmissiveSmoke {type=NiagaraBool}
 ]
}
